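# Physics asset for the Rebirth Medic backpack cosmetic (F_MED), giving collision bodies to the pack and its dangling cable and zipper bones.
# A box covers the root, while small capsules sit on pack, dyn_cable and dyn_zipper.
# Two constraints tie the dyn bones to pack: the cable's swing and twist are locked, while the zipper has its own uneven swing angles and a tight twist angle.
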PhysicsAsset /BRCosmetics/Accessories/FORT_Backpacks/Backpack_F_MED_Rebirth_Medic/Meshes/F_MED_Rebirth_Medic_Physics
bodies {
 dyn_cable [{shapeType=Capsule, length=1, radius=1}]
 dyn_zipper [{shapeType=Capsule, length=0, radius=1, center=<2 0 0>}]
 pack [{shapeType=Capsule, length=1, radius=1}]
 root [{shapeType=Box, extentX=7.5, extentY=13.5, extentZ=15.5, center=<-6.4534 -6.3737 -4.0535>, rotation=<rot -9.0965 20.186 -5.1441>}]
}
constraints [
 {
  bone1Name=dyn_cable
  bone2Name=pack
  swing1LimitDegrees=45
  swing1Motion=Locked
  swing2LimitDegrees=45
  swing2Motion=Locked
  twistLimitDegrees=45
  twistMotion=Locked
 }
 {bone1Name=dyn_zipper, bone2Name=pack, swing1LimitDegrees=20, swing2LimitDegrees=45, twistLimitDegrees=2.5}
]
